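; Blueprint interface for elimination bot visual effects, placed with the death gameplay cue notifies.
; It declares two functions, TriggerVFX and EndVFX, both without parameters or bodies, so implementing assets supply the effect start and stop behavior.
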(fn EndVFX)

(fn TriggerVFX)
